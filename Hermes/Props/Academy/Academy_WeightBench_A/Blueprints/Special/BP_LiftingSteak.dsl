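; BP_LiftingSteak is an interactable steak prop for the Academy weight bench that a player can eat.
; Interaction is available only while CanEatSteak is set, showing an "Eat Steak" prompt with a short timed interaction.
; On interact, a DoOnce passes the pawn's Athena player controller to the weights manager's SteakEaten, then silently destroys the steak via SilentDie.
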(event Interaction|EventBlueprintOnInteract
  (params (InteractingPawn :type Fort%20Pawn%20Object%20Reference) (InteractionBeingAttempted :type EInteractionBeingAttempted%20Enum :default "FirstInteraction"))
  (Utilities|FlowControl|DoOnce false
    (:Completed
      (bind _asfort_player_controller_athena (Utilities|Casting|CastToFortPlayerControllerAthena (Pawn|GetController InteractingPawn)))
      (Class|BPWeightsManager|SteakEaten (Variables|Default|GetManager) _asfort_player_controller_athena)
      (Gameplay|SilentDie false))))

(fn BlueprintCanInteract
  (params (InteractingPawn :type Fort%20Pawn%20Object%20Reference) (InteractionBeingAttempted :type EInteractionBeingAttempted%20Enum :default "FirstInteraction") (InteractionType :type TInteractionType%20Enum :default "IT_NoInteraction"))
  (return (Variables|Default|GetCanEatSteak)))

(fn BlueprintGetInteractionString
  (params (InteractingPawn :type Fort%20Pawn%20Object%20Reference) (InteractionBeingAttempted :type EInteractionBeingAttempted%20Enum :default "FirstInteraction"))
  (return "NSLOCTEXT(\"[7812009ABEA389E01620DB6BBF2D276D]\", \"2E74902F429CCD768513C98A2611F328\", \"Eat Steak \")"))

(fn BlueprintGetInteractionTime
  (params (InteractingPawn :type Fort%20Pawn%20Object%20Reference) (InteractionBeingAttempted :type EInteractionBeingAttempted%20Enum :default "FirstInteraction"))
  (return 1.0 true))

(fn ConstructionScript)
