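; Interactable 515 KG deadlift prop that spawns "deadlift record" loot when a player finishes the timed lift, then destroys itself.
; On interact start, it shortens the lift time per steak recorded in BP_WeightsManager, with a one-second floor on the final value.
; The steak count equals the manager's whole list length whenever the player appears in it.
; Completion tosses pickups randomly and plays chest effects on non-authority only, guarded by DoOnce.
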(event EventBeginPlay
  (Variables|Default|SetLiftDurationActive (Variables|Default|GetLiftDurationDefault)) ; 2 mins mins without steaks, 59 seconds quicker each stake
)

(event Interaction|EventBlueprintOnInteract
  (params (InteractingPawn :type Fort%20Pawn%20Object%20Reference) (InteractionBeingAttempted :type EInteractionBeingAttempted%20Enum :default "FirstInteraction"))
  (Utilities|FlowControl|DoOnce false
    (:Completed
      (CallFunction|SpawnLoot))))

(event Interaction|EventBlueprintOnBeginInteract
  (params (InteractingPawn :type Fort%20Pawn%20Object%20Reference) (InteractionBeingAttempted :type EInteractionBeingAttempted%20Enum :default "FirstInteraction") (InteractionType :type TInteractionType%20Enum :default "IT_NoInteraction"))
  (bind _asfort_player_controller_athena (Utilities|Casting|CastToFortPlayerControllerAthena (Pawn|GetController (Variables|/Script/Engine.Actor|GetInstigator InteractingPawn))))
  (bind _numsteaks (CallFunction|ReturnNumOfSteaksforController _asfort_player_controller_athena))
  (bind _newtime (CallFunction|SetInteractDurationBasedonSteaks _numsteaks))
  (Variables|Default|SetLiftDurationActive _newtime))

(event Custom|SpawnLoot
  (Utilities|FlowControl|DoOnce false
    (:Completed
      (Utilities|FlowControl|Sequence
        (:then_0
          (CallFunction|PlayLiftEffects)
          (bind _returnvalue (Inventory|SpawnPickupInWorldWithLootTier "LTG_DeadliftRecord" (+ (Transformation|GetActorLocation self) "0.000000,0.000000,125.000000") 0 0 true true "100.000000,100.000000,100.000000"))
          (for _array_element _returnvalue
            (bind _asfort_pickup_athena (Utilities|Casting|CastToFortPickupAthena _array_element))
            (Pickup|TossPickupinDirection _asfort_pickup_athena (Math|Vector|MakeVector (* 1.0 (Math|Random|RandomFloatinRange -4e+02 4e+02)) (* (Math|Random|RandomFloatinRange -4e+02 4e+02) 1.0) (* (Math|Random|RandomFloatinRange 1e+02 2.5e+02) 1.0)) self false false)))
        (:then_1
          (Utilities|FlowControl|Delay 0.25)
          (Actor|DestroyActor))))))

(event Custom|PlayLiftEffects
  (Utilities|FlowControl|DoOnce false
    (:Completed
      (Networking|SwitchHasAuthority
        (:Remote
          (Niagara|SpawnSystematLocation "/Game/Effects/Fort_Effects/Gameplay/PirateChests/NS_Loot_Chest.NS_Loot_Chest" (+ (Transformation|GetActorLocation self) "0.000000,0.000000,70.000000"))
          (Audio|PlaySoundatLocation "/Game/Sounds/Foley_Loot/Containers/Treasure_Chest/Tiered_Chest_Open_T01_Cue.Tiered_Chest_Open_T01_Cue" (+ (Transformation|GetActorLocation self) "0.000000,0.000000,70.000000")))))))

(fn BlueprintCanInteract
  (params (InteractingPawn :type Fort%20Pawn%20Object%20Reference) (InteractionBeingAttempted :type EInteractionBeingAttempted%20Enum :default "FirstInteraction") (InteractionType :type TInteractionType%20Enum :default "IT_NoInteraction"))
  (return true))

(fn BlueprintGetInteractionString
  (params (InteractingPawn :type Fort%20Pawn%20Object%20Reference) (InteractionBeingAttempted :type EInteractionBeingAttempted%20Enum :default "FirstInteraction"))
  (return "NSLOCTEXT(\"[C1E33E7C6ED654BF65E2604E8E9D5F34]\", \"ACB8404B4C2327121CF97D8DCF31C248\", \"Attempt Lift - 515 KG\")"))

(fn BlueprintGetInteractionTime
  (params (InteractingPawn :type Fort%20Pawn%20Object%20Reference) (InteractionBeingAttempted :type EInteractionBeingAttempted%20Enum :default "FirstInteraction"))
  (return (Variables|Default|GetLiftDurationActive) true))

(fn ReturnNumOfSteaksForController
  (params (Controller :type Fort%20Player%20Controller%20Athena%20Object%20Reference))
  (bind _manager (Variables|Default|GetManager))
  (bind _allcontrollerswhohaveeatensteaks (Variables|/Hermes/Props/Academy/Academy_WeightBench_A/Blueprints/Special/BP_WeightsManager.BP_WeightsManager_C|GetAllControllersWhoHaveEatenSteaks _manager))
  (for _ _allcontrollerswhohaveeatensteaks
    (if (Utilities|Array|ContainsItem _allcontrollerswhohaveeatensteaks Controller)
      (Math|Integer|IncrementInt (LocalVariables|GetNumContained))))
  (return (LocalVariables|GetNumContained)))

(fn SetInteractDurationBasedOnSteaks
  (params (NumSteaks :type Integer))
  (bind _output_get (Variables|Default|SetLiftDurationActive (- (Variables|Default|GetLiftDurationDefault) (* (LocalVariables|GetTimePerSteakToRemove) NumSteaks))))
  (return (Math|Float|Clamp(Float) _output_get 1.0 9999.0)))

(fn ConstructionScript
  (|CallParentFunction))
