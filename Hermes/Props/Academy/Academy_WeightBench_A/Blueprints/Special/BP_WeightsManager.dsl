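; Weights manager Blueprint for the Academy weight bench prop that records players who eat steaks.
; A SteakEaten custom event takes an Athena player controller and adds it to an array of controllers that have eaten steaks.
; After the add it reaches a DoOnce with nothing following it, and the construction script is empty.
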(event Custom|SteakEaten
  (params (Controller :type Fort%20Player%20Controller%20Athena%20Object%20Reference))
  (Utilities|Array|Add (Variables|Default|GetAllControllersWhoHaveEatenSteaks) Controller)
  (Utilities|FlowControl|DoOnce false))

(fn ConstructionScript)
